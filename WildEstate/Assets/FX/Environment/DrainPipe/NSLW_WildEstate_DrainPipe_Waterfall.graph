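# Niagara system for a drain pipe waterfall environment effect in WildEstate, with a user Global Scale float.
# It holds two Splash emitters and an Empty002 emitter, none of which list modules or renderers here.
# A disabled FN_Mesh emitter burst-spawns tinted, offset and rotated mesh particles that persist past their lifetime, with infinitely lived ones allowed to die when the emitter deactivates.
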NiagaraSystem /WildEstate/Assets/FX/Environment/DrainPipe/NSLW_WildEstate_DrainPipe_Waterfall
dependencies {
 usedDataInterfaces []
 usedDynamicInputs [
  &Add_Vector
  /Niagara/DynamicInputs/Helpers/SimulationPosition
  &Multiply_VectorByFloat
 ]
 usedModules [&SystemState]
}
emitters {
 Empty002 {rendererClasses [], renderers []}
 FN_Mesh {
  bEnabled=false
  emitterUpdateScript [
   &EmitterState()
   &SpawnBurst_Instantaneous(Use Spawn Probability=false, UseLoopCountLimit=true)
  ]
  particleSpawnScript [
   &InitializeParticle(Color=<rgba 0.5 0.7924 1 1>, Position Offset=<0 0 -50>, UsePositionOffset=true, Write Material Random=true) {module=InitializeParticle001}
   &InitialMeshOrientation(Mesh Orientation Mode=::NewEnumerator1, Rotation=<0 0 0.25>, Use Rotation Vector=true)
  ]
  particleUpdateScript [&ParticleState(Kill Particles When Lifetime Has Elapsed=false, Let Infinitely Lived Particles Die When Emitter Deactivates=true)]
  rendererClasses [&NiagaraMeshRendererProperties]
  renderers [{rendererClass=&NiagaraMeshRendererProperties, rendererIndex=0}]
 }
 Splash {rendererClasses [], renderers []}
 Splash001 {rendererClasses [], renderers []}
}
summary {
 systemName=NSLW_WildEstate_DrainPipe_Waterfall
 emitters [
  {emitterName=Splash, rendererClasses []}
  {emitterName=Splash001, rendererClasses []}
  {emitterName=Empty002, rendererClasses []}
  {bEnabled=false, emitterName=FN_Mesh, rendererClasses [&NiagaraMeshRendererProperties]}
 ]
 userVariables [User.Global Scale {type=NiagaraFloat, defaultValue=1, description=""}]
}
